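# Material function for the Eclipse Vault elevator door that clips the mesh beyond a fixed width around a reference position, likely hiding panels as they slide.
# It takes world position minus an ActorPosition vector from custom primitive data, rotates that offset into local space and takes absolute X.
# A hard-coded half-width threshold becomes an inverted, saturated Opacity Mask, written onto the incoming material attributes.
# The function itself contains no time input or movement.
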MaterialFunction /Game/Environments/Asteria/Sets/EclipseVault/Materials/MF_ElevatorDoorAnimation
graph {nodes [
  n1 = FunctionInput() {inputType=FunctionInput_MaterialAttributes}
  n2 = WorldPosition()
  n3 = VectorParameter() {bUseCustomPrimitiveData=true, parameterName=ActorPosition, defaultValue=<rgba 0 0 0 1>}
  n4 = ComponentMask(None=Transform(None=Subtract(A=@n2.XYZ, B=@n3.RGB)) {transformSourceType=TRANSFORMSOURCE_World, transformType=TRANSFORM_Local}) {r=true}
  n5 = Saturate(None=Add(A=Abs(None=@n4), B=Constant() {r=-150}))
  FunctionOutput(None=SetMaterialAttributes!(MaterialAttributes=GetMaterialAttributes(None=@n1).MaterialAttributes, OpacityMask=OneMinus(None=@n5)) {attributeSetTypes [#1]})
 ]}
