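# Material function for the Eclipse Vault door that spins its handle and fades out emissive once triggered.
# A custom primitive data start time, TimeInSeconds, drives a clamped lookup into a grayscale animation texture over 13.5 seconds.
# That value rotates vertices about the local Y axis at VaultHandleCenter via world position offset, blended in by red vertex color.
# The same start time fades emissive to zero over one second; zero leaves both effects at rest.
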MaterialFunction /Game/Environments/Asteria/Sets/EclipseVault/Materials/MF_VaultDoorAnimation
graph {nodes [
  n1 = GetMaterialAttributes(None=FunctionInput() {inputType=FunctionInput_MaterialAttributes}) {attributeGetTypes [#1, #2]}
  n2 = ScalarParameter() {bUseCustomPrimitiveData=true, parameterName=TimeInSeconds}
  n3 = Subtract(A=Time(), B=@n2)
  n4 = Divide(A=Abs(None=@n3)) {constB=13.5}
  n5 = Saturate(None=@n2)
  n6 = TextureSample(UVs=Multiply(A=Saturate(None=@n4), B=@n5)) {
   samplerSource=SSM_Clamp_WorldGroupSettings
   samplerType=SAMPLERTYPE_LinearGrayscale
   texture=/Game/Environments/Asteria/Sets/EclipseVault/Textures/T_VaultHeist_HandleAnimation
  }
  n7 = MaterialFunctionCall!(
   RotationAngle=Multiply(A=@n6.R) {constB=-10}
   NormalizedRotationAxis=Transform(None=Constant3Vector() {constant=<rgba 0 1 0 0>}) {transformSourceType=TRANSFORMSOURCE_Local}
   PivotPoint=TransformPosition(None=VectorParameter() {parameterName=VaultHandleCenter, defaultValue=<rgba 179.7 -80 -42 1>}.RGB) {transformType=TRANSFORMPOSSOURCE_World}
   Position=WorldPosition().XYZ
   Normal=@n1.Normal
   TangentU
   TangentV
  ) {materialFunction=/Game/Characters/Player/Male/Medium/Bodies/M_MED_Inferno/FX/Mat/RotateVertexAboutAxis}
  n8 = SetMaterialAttributes!(MaterialAttributes=@n1.MaterialAttributes, WorldPositionOffset=@n7.WPO, Normal=@n7.WPO) {attributeSetTypes [#3, #1]}
  n9 = VertexColor()
  n10 = NamedRerouteUsage()
  FunctionOutput(None=SetMaterialAttributes!(
    MaterialAttributes=BlendMaterialAttributes(A=@n1.MaterialAttributes, B=@n8, Alpha=@n9.R)
    EmissiveColor=Multiply(A=@n1.EmissiveColor, B=@n10)
   ) {attributeSetTypes [#2]})
  n11 = Saturate(None=Divide(A=@n3) {constB=1})
  NamedRerouteDeclaration(None=OneMinus(None=Multiply(A=@n5, B=@n11))) {name=Emission Fade, nodeColor=<rgba 0 1 0.8125 1>}
 ]}
