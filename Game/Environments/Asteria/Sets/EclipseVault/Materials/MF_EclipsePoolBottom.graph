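# Material function that paints a fake-depth pool bottom for the Eclipse Vault set, overriding base color, metallic, roughness, emissive and normal.
# Camera rays intersect a virtual box (interior-mapping style), providing UVs for two noise-warped, panning caustic layers and a depth-based texture gradient.
# Caustics, tint, a two-color view gradient and Fresnel blend into emissive with inverse eye adaptation.
# Quality and shading-path switches adjust roughness and emissive strength for lower tiers, forward and mobile.
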MaterialFunction /Game/Environments/Asteria/Sets/EclipseVault/Materials/MF_EclipsePoolBottom
graph {nodes [
  n1 = Constant()
  n2 = Constant() {r=0.5}
  n3 = QualitySwitch(Default=@n1, Low=@n2, High=@n1, Medium=@n2, Epic=@n1)
  n4 = NamedRerouteUsage()
  n5 = Multiply(A=CameraVectorWS(), B=Constant3Vector() {constant=<rgba 1 1 -1 0>})
  n6 = AppendVector(
   A=ScalarParameter() {defaultValue=100, group=PoolBottom, parameterName=SizeX}
   B=ScalarParameter() {defaultValue=100, group=PoolBottom, parameterName=SizeY}
  )
  n7 = Constant() {r=500}
  n8 = Multiply(A=@n5, B=Divide(B=AppendVector(A=@n6, B=@n7)))
  n9 = Divide(B=@n8) {desc=id}
  n10 = Subtract(
   A=ComponentMask(None=WorldPosition().XYZ) {g=true, r=true}
   B=AppendVector(A=ScalarParameter() {group=PoolBottom, parameterName=PositionX}, B=ScalarParameter() {group=PoolBottom, parameterName=PositionY})
  )
  n11 = Divide(A=NamedRerouteDeclaration(None=VertexInterpolator(VS=@n10).PS) {name=Texcoord, nodeColor=<rgba 1 0 0.3516 1>}, B=@n6)
  n12 = AppendVector(A=Subtract(A=Multiply(A=@n11) {constB=-2}), B=Constant() {r=-1})
  n13 = MaterialFunctionCall(Float3=Subtract(A=Abs(None=@n9), B=Multiply(A=@n9, B=@n12)) {desc=k}) {materialFunction=&BreakOutFloat3Components}
  n14 = Min(A=@n13.R, B=@n13.R)
  n15 = Add(A=Multiply(A=Min(A=@n14, B=@n13.B), B=@n8), B=@n12) {desc="pos.xyz += kMin * In.dir;"}
  n16 = Multiply(A=ScalarParameter() {defaultValue=100, group=PoolBottom, parameterName=SizeX}, B=ComponentMask(None=@n15) {r=true})
  n17 = Multiply(A=ComponentMask(None=@n15) {b=true}, B=ScalarParameter() {defaultValue=50, group=PoolBottom, parameterName=SizeZ})
  n18 = ScalarParameter() {defaultValue=2400, parameterName=CausticsScale}
  n19 = Multiply(A=ScalarParameter() {defaultValue=100, group=PoolBottom, parameterName=SizeY}, B=ComponentMask(None=@n15) {g=true})
  n20 = ComponentMask(None=Abs(None=@n15)) {r=true}
  n21 = LinearInterpolate(
   A=Divide(A=AppendVector(A=@n16, B=@n17), B=@n18) {constB=1000}
   B=Divide(A=AppendVector(A=@n19, B=@n17), B=@n18) {constB=1000}
   Alpha=Floor(None=Add(A=@n20) {constB=0.01})
  )
  n22 = Time() {bOverride_Period=true, period=50}
  n23 = NamedRerouteUsage()
  n24 = Subtract(A=Add(A=@n21, B=Multiply(A=@n22) {constB=0.02}), B=@n23)
  n25 = Multiply(A=@n15) {constB=0.5}
  n26 = Add(A=ComponentMask(None=Add(A=@n25) {constB=0.5}) {b=true}, B=ComponentMask(None=NamedRerouteUsage()) {b=true})
  n27 = OneMinus(None=Multiply(A=@n26, B=@n26))
  n28 = Multiply(A=@n27, B=@n27)
  n29 = Constant2Vector() {g=-0.02, r=0.02}
  n30 = Add(A=@n23, B=Add(A=Multiply(A=@n21) {constB=0.8}, B=Multiply(A=@n22, B=@n29) {constB=0.02}))
  n31 = Add(
   A=Multiply(A=TextureSample(UVs=@n24) {samplerType=SAMPLERTYPE_LinearColor, texture=/FortWater/Textures/T_Caustic_M}.G, B=@n28)
   B=Multiply(A=TextureSample(UVs=@n30) {samplerType=SAMPLERTYPE_LinearColor, texture=&T_Caustic_M}.G, B=@n28)
  )
  n32 = Multiply(
   A=Multiply(A=Multiply(A=@n31, B=@n31)) {constB=0.1}
   B=VectorParameter() {parameterName=ColorCaustics, defaultValue=<rgba 0 0.111 0.2031 1>}.RGB
  )
  n33 = Saturate(None=@n26)
  n34 = TextureSample(UVs=OneMinus(None=@n33)) {
   samplerSource=SSM_Clamp_WorldGroupSettings
   samplerType=SAMPLERTYPE_LinearColor
   texture=/Game/Environments/Asteria/Sets/EclipseVault/Textures/T_EclipsePoolBottom_D
  }
  n35 = Multiply(A=VectorParameter() {parameterName=Color, defaultValue=<rgba 0.0074 0.0517 0.0885 1>}.RGB, B=@n34.RGB)
  n36 = VectorParameter() {parameterName=CubemapColor1, defaultValue=<rgba 0.0022 0.0147 0.0417 1>}
  n37 = VectorParameter() {parameterName=CubemapColor2, defaultValue=<rgba 0 0.4736 1 1>}
  n38 = ComponentMask(None=Add(A=CameraVectorWS(), B=NamedRerouteUsage())) {g=true, r=true}
  n39 = Multiply(A=@n38, B=@n38)
  n40 = Multiply(A=Add(A=ComponentMask(None=@n39) {r=true}, B=ComponentMask(None=@n39) {g=true})) {constB=0.14}
  n41 = CameraVectorWS()
  n42 = Constant3Vector() {constant=<rgba 0 0 -1 0>}
  n43 = OneMinus(None=Saturate(None=DotProduct(A=@n41, B=@n42)))
  n44 = LinearInterpolate(
   A=LinearInterpolate(A=Add(A=@n32, B=@n35), B=LinearInterpolate(A=@n36.RGB, B=@n37.RGB, Alpha=@n40), Alpha=SmoothStep(Value=@n33) {constMin=0.5})
   B=VectorParameter() {parameterName=ColorFresnel, defaultValue=<rgba 0.0007 0.0247 0.0677 1>}.RGB
   Alpha=Multiply(A=@n43, B=@n43)
  )
  n45 = EyeAdaptationInverse(LightValueInput=Multiply(A=Add(A=@n4, B=@n44), B=@n34.A), AlphaInput=Constant() {r=0.5})
  n46 = Multiply(A=@n45.EyeAdaptationInverse) {constB=1.3}
  n47 = QualitySwitch(Default=@n45.EyeAdaptationInverse, Low=@n46, High=@n45.EyeAdaptationInverse, Medium=@n46, Epic=@n45.EyeAdaptationInverse)
  n48 = SetMaterialAttributes!(
   MaterialAttributes=GetMaterialAttributes(None=FunctionInput() {inputType=FunctionInput_MaterialAttributes}).MaterialAttributes
   BaseColor=VectorParameter() {parameterName=BaseColor, defaultValue=<rgba 0.0031 0.1082 0.2969 1>}.RGB
   Metallic=Constant() {r=0.8}
   Roughness=ShadingPathSwitch(Default=@n3, Deferred=@n3, Forward=@n2, Mobile=@n2)
   EmissiveColor=ShadingPathSwitch(Default=@n47, Deferred=@n47, Forward=@n46, Mobile=@n46)
   Normal=Constant3Vector() {constant=<rgba 0 0 -1 0>}
  ) {attributeSetTypes [
    #1
    #2
    #3
    #4
    #5
   ]}
  FunctionOutput(None=@n48)
  n49 = ScalarParameter() {defaultValue=0.5, parameterName=WarpTiling}
  n50 = Time() {bOverride_Period=true, period=20}
  n51 = TextureSample(UVs=Add(A=Multiply(A=@n21, B=@n49), B=Multiply(A=@n50) {constB=0.05})) {samplerType=SAMPLERTYPE_LinearColor, texture=&HighContrastNoise}
  Multiply(A=@n51.RGB, B=ScalarParameter() {defaultValue=0.1, parameterName=CubeWarp})
  n52 = Add(A=@n51.RGB) {constB=-0.5}
  NamedRerouteDeclaration(None=Multiply(A=ComponentMask(None=@n52) {g=true, r=true}, B=ScalarParameter() {defaultValue=0.3, parameterName=CausticsWarp})) {name=CausticsWarp, nodeColor=<rgba 0.125 0 1 1>}
  n53 = Constant() {r=1164}
  n54 = AppendVector(A=@n53, B=@n53)
  n55 = ScalarParameter() {defaultValue=50, group=PoolBottom, parameterName=WaterSurfaceZ}
  n56 = Multiply(A=@n5, B=Divide(B=AppendVector(A=@n54, B=@n55)))
  n57 = Divide(B=@n56) {desc=id}
  n58 = MaterialFunctionCall(Float3=Subtract(A=Abs(None=@n57), B=Multiply(A=@n57, B=@n12)) {desc=k}) {materialFunction=&BreakOutFloat3Components}
  n59 = Min(A=@n58.R, B=@n58.R)
  n60 = Add(A=Multiply(A=Min(A=@n59, B=@n58.B), B=@n56), B=@n12) {desc="pos.xyz += kMin * In.dir;"}
  n61 = Saturate(None=Add(A=ComponentMask(None=@n60) {b=true}) {constB=0.9})
  n62 = Multiply(A=@n61, B=@n61)
  n63 = ComponentMask(None=@n60) {g=true, r=true}
  n64 = Time() {bOverride_Period=true, period=20}
  n65 = TextureSample(UVs=Add(A=Multiply(A=@n63) {constB=0.4}, B=Multiply(A=@n64) {constB=0.05})) {samplerType=SAMPLERTYPE_LinearColor, texture=&HighContrastNoise}
  NamedRerouteDeclaration(None=Multiply(A=Multiply(A=@n62, B=@n65.RGB), B=ScalarParameter() {defaultValue=0.1, parameterName=CubeWarp})) {name=CubeWarp, nodeColor=<rgba 1 0.75 0 1>}
  n66 = Divide(A=NamedRerouteUsage()) {constB=2000}
  n67 = Multiply(A=@n66) {constB=2}
  n68 = Time() {bOverride_Period=true, period=71.4286}
  n69 = Multiply(A=@n68) {constB=0.014}
  n70 = Time() {bOverride_Period=true, period=40}
  n71 = TextureSample(UVs=Add(A=Multiply(A=@n66) {constB=0.5}, B=Multiply(A=@n70) {constB=0.025})) {samplerType=SAMPLERTYPE_LinearColor, texture=&HighContrastNoise}
  n72 = Multiply(A=ComponentMask(None=Add(A=@n71.RGB) {constB=-0.5}) {g=true, r=true}, B=ScalarParameter() {defaultValue=0.3, parameterName=CausticsWarp})
  n73 = TextureSample(UVs=Subtract(A=Add(A=@n67, B=@n69), B=@n72) {constB=0}) {samplerType=SAMPLERTYPE_LinearColor, texture=&T_Caustic_M}
  n74 = Multiply(A=@n67) {constB=0.8}
  n75 = Multiply(A=@n68, B=Constant2Vector() {g=-0.014, r=0.014}) {constB=0.02}
  n76 = TextureSample(UVs=Add(A=@n72, B=Add(A=@n74, B=@n75))) {samplerType=SAMPLERTYPE_LinearColor, texture=&T_Caustic_M}
  n77 = Add(A=Max(A=@n73.R, B=@n73.R), B=Max(A=@n76.R, B=@n76.R))
  n78 = Multiply(A=@n77, B=@n77)
  n79 = ScalarParameter() {defaultValue=0.12, parameterName=CausticsIntensity}
  NamedRerouteDeclaration(None=Multiply(A=Multiply(A=@n78, B=@n79) {constB=0.1}, B=VectorParameter() {parameterName=ColorCaustics, defaultValue=<rgba 0 0.111 0.2031 1>}.RGB)) {name=GlassCaustics, nodeColor=<rgba 0 0.1328 1 1>}
  ScalarParameter() {defaultValue=0.14, parameterName=CausticsIntensity}
 ]}
