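# M_EclipsePoolBottomLite is a material for the pool bottom in Asteria's EclipseVault environment set.
# It takes all its material attributes from a single MF_EclipsePoolBottom function call, so the surface logic lives in that function.
# It uses Default Lit shading, is flagged for Nanite and treats normals as non-tangent-space.
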Material /Game/Environments/Asteria/Sets/EclipseVault/Materials/M_EclipsePoolBottomLite
graph {
 properties {
  bTangentSpaceNormal=false
  bUseMaterialAttributes=true
  bUsedWithNanite=true
  pixelDepthOffsetMode=PDOM_Legacy
  usedShadingModels=Default Lit
 }
 nodes [MaterialFunctionCall!(In=GetMaterialAttributes().MaterialAttributes) {materialFunction=&/MF_EclipsePoolBottom}]
}
