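# Lit, Nanite-enabled Eclipse Vault screen material for a lockdown display, with animated wave lines in red by default.
# Two channels of a Lazarus Lens face SDF, scaled and offset by custom primitive data, are levels-remapped into time-scrolled sine bands, one with an oscillating phase.
# The combined bands are tinted, then modulated by noise, tiled LCD detail, panel seams and a per-panel gradient.
# EyeAdaptationInverse compensates the glow for exposure.
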Material /Game/Environments/Asteria/Sets/EclipseVault/Materials/M_EclipseVault_Screen_Lockdown
graph {
 properties {bUsedWithNanite=true, pixelDepthOffsetMode=PDOM_Legacy, usedShadingModels=Default Lit}
 nodes [
  n1 = Sine(None=Time()) {period=10}
  ConstantBiasScale(None=@n1)
  n2 = Frac(None=Add(A=TextureCoordinate() {uTiling=3, vTiling=3}) {constB=1.5})
  n3 = MaterialFunctionCall(Float2=Abs(None=ConstantBiasScale(None=@n2) {bias=-0.5, scale=2})) {materialFunction=&BreakOutFloat2Components}
  n4 = SphereMask(A=Min(A=@n3.R, B=@n3.R), B=Constant() {r=1}, Radius=Constant() {r=0.985})
  n5 = TextureCoordinate()
  n6 = ScalarParameter() {defaultValue=180, parameterName=LCD_Repeat}
  n7 = Multiply(
   A=@n4
   B=TextureSample(UVs=Multiply(A=@n5, B=@n6)) {samplerType=SAMPLERTYPE_LinearColor, texture=/Game/Environments/Asteria/Sets/EclipseVault/Textures/Screens/T_ScreenDetails_D}.RGB
  )
  n8 = TextureCoordinate()
  n9 = AppendVector(
   A=ScalarParameter() {bUseCustomPrimitiveData=true, parameterName=ScaleU}
   B=ScalarParameter() {bUseCustomPrimitiveData=true, parameterName=ScaleV, primitiveDataIndex=1}
  )
  n10 = ScalarParameter() {bUseCustomPrimitiveData=true, defaultValue=1, parameterName=OffsetU, primitiveDataIndex=2}
  n11 = ScalarParameter() {bUseCustomPrimitiveData=true, defaultValue=1, parameterName=OffsetV, primitiveDataIndex=3}
  n12 = TextureSample(UVs=Add(A=Multiply(A=@n8, B=@n9), B=AppendVector(A=@n10, B=@n11))) {
   samplerType=SAMPLERTYPE_LinearColor
   texture=/Game/Characters/Player/Male/Medium/Bodies/M_MED_LazarusLens_Style/Textures/T_LazarusLens_Style_FaceAcc_SDF
  }
  n13 = Constant() {r=0.3}
  n14 = Constant() {r=1.2}
  n15 = Constant() {r=4}
  n16 = Constant() {r=0.2}
  n17 = StaticBool() {value=true}
  n18 = Power(
   Base=Add(A=@n12.G, B=ScalarParameter() {defaultValue=0.1, parameterName=WaveFrequence_Pinch_Offset2}) {constB=0.1}
   Exp=ScalarParameter() {defaultValue=16, parameterName=WaveFrequence_Pinch_Power}
  ) {constExponent=30}
  n19 = LinearInterpolate(
   A=ScalarParameter() {defaultValue=0.3, parameterName=WaveFrequence_Base}
   B=ScalarParameter() {defaultValue=0.2, parameterName=WaveFrequence_Pinch}
   Alpha=MaterialFunctionCall(In=OneMinus(None=@n18), Contrast=ScalarParameter() {defaultValue=0.5, parameterName=WaveFrequence_Pinch_Contrast}) {materialFunction=&CheapContrast}.Result
  ) {constA=0.5}
  n20 = ScalarParameter() {defaultValue=1, parameterName=WaveFrequence_Eye}
  n21 = Power(Base=@n12.G, Exp=ScalarParameter() {defaultValue=6, parameterName=WaveFrequence_Eye_Power}) {constExponent=30}
  n22 = ScalarParameter() {defaultValue=200, parameterName=WaveFrequency}
  n23 = Multiply(
   A=MaterialFunctionCall!(
    Texture=@n12.G
    ---------------
    New Black Value=@n13
    New Middle Value=@n14
    New White Value=@n15
    Middle Point=@n16
    "--------------- "
    Define Interpolation Curve=@n17
    Interpolation Power=LinearInterpolate(A=@n19, B=@n20, Alpha=@n21) {constA=0.5}
    Invert Interpolation Power
   ) {materialFunction=/Engine/Functions/Engine_MaterialFunctions02/3PointLevels}.Result
   B=@n22
  ) {constB=1000}
  n24 = Multiply(A=Time(), B=ScalarParameter() {defaultValue=30, parameterName=WaveSpeed})
  n25 = ConstantBiasScale(None=Sine(None=Add(A=@n23, B=@n24)) {period=100})
  n26 = ScalarParameter() {defaultValue=0.18, parameterName=WaveThickness}
  n27 = ScalarParameter() {defaultValue=5, parameterName=WaveHardness}
  n28 = MaterialFunctionCall(In=Power(Base=Saturate(None=@n25), Exp=@n26) {constExponent=0.2}, Contrast=@n27) {materialFunction=&CheapContrast}
  n29 = Add(A=@n24, B=Multiply(A=ScalarParameter() {defaultValue=50, parameterName=WaveOffset}, B=@n1))
  n30 = Power(
   Base=Add(A=@n12.A, B=ScalarParameter() {defaultValue=0.1, parameterName=WaveFrequence_Pinch_Offset2}) {constB=0.1}
   Exp=ScalarParameter() {defaultValue=16, parameterName=WaveFrequence_Pinch_Power}
  ) {constExponent=30}
  n31 = LinearInterpolate(
   A=ScalarParameter() {defaultValue=0.3, parameterName=WaveFrequence_Base}
   B=ScalarParameter() {defaultValue=0.2, parameterName=WaveFrequence_Pinch}
   Alpha=MaterialFunctionCall(In=OneMinus(None=@n30), Contrast=ScalarParameter() {defaultValue=0.5, parameterName=WaveFrequence_Pinch_Contrast}) {materialFunction=&CheapContrast}.Result
  ) {constA=0.5}
  n32 = ScalarParameter() {defaultValue=1, parameterName=WaveFrequence_Eye}
  n33 = Power(Base=@n12.A, Exp=ScalarParameter() {defaultValue=6, parameterName=WaveFrequence_Eye_Power}) {constExponent=30}
  n34 = Multiply(
   A=MaterialFunctionCall(
    Texture=@n12.A
    New Black Value=@n13
    New Middle Value=@n14
    New White Value=@n15
    Middle Point=@n16
    Define Interpolation Curve=@n17
    Interpolation Power=LinearInterpolate(A=@n31, B=@n32, Alpha=@n33) {constA=0.5}
   ) {materialFunction=&3PointLevels}.Result
   B=@n22
  ) {constB=1000}
  n35 = ConstantBiasScale(None=Sine(None=Add(A=@n29, B=@n34)) {period=100})
  n36 = MaterialFunctionCall(In=Power(Base=Saturate(None=@n35), Exp=@n26) {constExponent=0.2}, Contrast=@n27) {materialFunction=&CheapContrast}
  n37 = Multiply(A=@n4, B=Max(A=OneMinus(None=@n28.Result), B=OneMinus(None=@n36.Result)))
  n38 = VectorParameter() {parameterName=MinColor, defaultValue=<rgba 0.04 0.04 0.04 0>}
  n39 = TextureCoordinate() {uTiling=1.5, unMirrorU=true, vTiling=1.5}
  n40 = Multiply(
   A=Multiply(
    A=Max(A=@n37, B=@n38.RGB)
    B=TextureSample(UVs=@n39) {samplerType=SAMPLERTYPE_Grayscale, texture=&LowResBlurredNoise}.R
   )
   B=VectorParameter() {parameterName=EmissiveColor, defaultValue=<rgba 25 0.1302 0.1302 0>}.RGB
  )
  n41 = ComponentMask(None=Frac(None=TextureCoordinate() {uTiling=3, vTiling=3})) {r=true}
  n42 = ScalarParameter() {defaultValue=0.3, parameterName=ScreenPanelBoost}
  EyeAdaptationInverse(LightValueInput=Multiply(A=Multiply(A=@n7, B=@n40), B=Add(A=@n41, B=@n42)))
  MaterialFunctionCall() {materialFunction=&MF_FN_DestructibleDamage}
  ScalarParameter() {defaultValue=0.15, parameterName=Roughness}
  ScalarParameter() {defaultValue=0.2, parameterName=Specular}
  ScalarParameter() {defaultValue=0.5, parameterName=Metallic}
  VectorParameter() {parameterName=BaseColor, defaultValue=<rgba 0.0521 0.0521 0.0521 0>}
 ]
}
